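; Exterior thermal sign prop in the DustMite set, with optional extra collision and server-side death handling.
; When SpawnExtraCollision is set, the construction script adds an offset, scaled box collision component and stores it in BoxCollision; otherwise it destroys any stored one.
; On server death, it destroys the box and then the actor, but only if that flag is set and the box is valid.
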(event EventOnDeathServer
  (params (Damage :type Float%20%28single-precision%29) (DamageTags :type Gameplay%20Tag%20Container%20Structure%20%28by%20ref%29) (Momentum :type Vector :default "0, 0, 0") (HitInfo :type Hit%20Result%20Structure%20%28by%20ref%29) (InstigatedBy :type Controller%20Object%20Reference) (DamageCauser :type Actor%20Object%20Reference) (EffectContext :type Gameplay%20Effect%20Context%20Handle%20Structure))
  (bind _boxcollision (Variables|Default|GetBoxCollision))
  (if (Variables|Default|GetSpawnExtraCollision)
    (Utilities|IsValid _boxcollision
      (:Is%20Valid
        (Components|DestroyComponent _boxcollision)
        (Actor|DestroyActor)))))

(fn ConstructionScript
  (if (Variables|Default|GetSpawnExtraCollision)
    (bind _returnvalue (AddComponent|NODE_AddBoxComponent-0 :RelativeTransform (Math|Transform|MakeTransform "14.000000,-20.000000,0.000000" "0, 0, 0" "26.000000,1.800000,1.200000")))
    (Variables|Default|SetBoxCollision _returnvalue)
    (else
      (Components|DestroyComponent (Variables|Default|GetBoxCollision)))))
